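; Vivid Razor Vault alert light actor that toggles its flare, point light, looping alarm sound and emissive material with the alert state.
; UpdateAlertVisuals takes the active state from its parent function and drives those effects.
; On world ready, the authority links it to the environment below, retrying on a timer up to a maximum to cover late-loading level instance pieces.
; When the linked actor is destroyed, it silences the alarm and expires shortly after.
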(event EventOnWorldReady
  ;[0 1 0.135] EventOnWorldReady
  ; Try to link to the environment when the World is Ready.
  (|CallParentFunction)
  (Networking|SwitchHasAuthority
    (:Authority
      (CallFunction|TrytoLinktoEnvironment))))

(event Custom|TryToLinkToEnvironment
  ;[0.3798 0.6987 1] TryToLinkToEnvironment
  ; Attempt to attach the alarm to the environment to prevent it from floating if the env is destroyed
  (Math|Integer|IncrementInt (Variables|LinkToEnvironment|GetLinkAttemptsCount)
    (:%20%20
      (bind _returnvalue (LinkToActorComponent|PerformLinkingCheck (Variables|Default|GetFortLinkToActor) "Down" (Variables|LinkToEnvironment|GetEnvironmentLinkedActor)))
      ;;[1 0 0.8947] Fallback handles case where the alarm which is in a content bundle in Asteria is loaded before the environment piece it is trying to attach to (in a level instance). Linking to an indestructible actor (ex terrain) underneath will fail
      (if _returnvalue
        _
        (elif (>= (Variables|LinkToEnvironment|GetLinkAttemptsCount) (Variables|LinkToEnvironment|GetLinkAttemptsMax))
          (Fortnite|FortShippingLog self (Utilities|String|Append "VIVID RAZOR CHEST: Attempt to Link to Environment Failed on total attempts number:  " (Utilities|String|ToString(Integer) (Variables|LinkToEnvironment|GetLinkAttemptsMax))))
          (else
            (bind _returnvalue_2 (Utilities|Time|SetTimerbyEvent (EventDispatchers|CreateEvent :!event Custom|TryToLinkToEnvironment) (Variables|LinkToEnvironment|GetLinkAttemptTiming))) ; try to link again
            (Variables|LinkToEnvironment|SetTimerH_RetryLinkToEnvironment _returnvalue_2)))))))

(event OnLinkedActorDestroyed(FortLinkToActor)
  (params (DamageCauser :type Actor%20Object%20Reference))
  ;[0.4149 0.164 0] OnLinkedActorDestroyed
  ; Set the life span to destroy the alarm when the attached envionment actor is destoyed
  (SoundLibrary|StopEvent (Variables|Default|GetSoundLibraryComponent_VividRazorVault) (Variables|Default|GetAlarmSoundLibraryTag))
  (Actor|SetLifeSpan 0.2))

(event EventEndPlay
  (params (EndPlayReason :type EEndPlayReason%20Enum :default "Destroyed"))
  ;[1 0.3416 0] EndPlay
  ; Cleanup- clear timers, unbind events, etc.
  (Utilities|FlowControl|Sequence
    (:then_0
      (Utilities|Time|ClearandInvalidateTimerbyHandle (Variables|LinkToEnvironment|GetTimerH_RetryLinkToEnvironment)))
    (:then_1
      (|CallParentFunction EndPlayReason) ; call parent EndPlay last in case of \norder of operations dependencies
    )))

(fn GetTimeOfDayBlueprintDefaultVariables
  (return (Utilities|Struct|MakeTimeOfDayBlueprintDefaultVariables 0 0.0 (not (Variables|/Script/Engine.SceneComponent|GetbVisible (Variables|Default|GetPointLight))))))

(fn UpdateAlertVisuals
  (params (ForceOff :type Boolean) (NewState :type Boolean))
  (bind _staticmeshcomponent (Variables|Debug|GetStaticMeshComponent))
  (bind _material_instance_dynamic (Class|BTechArtGlobalLibrary|GetDynamicMaterialInstance _staticmeshcomponent))
  (bind _alarmsoundlibrarytag (Variables|Default|GetAlarmSoundLibraryTag))
  (bind _soundlibrarycomponent_vividrazorvault (Variables|Default|GetSoundLibraryComponent_VividRazorVault))
  (bind _isactive (|CallParentFunction ForceOff NewState))
  (Utilities|FlowControl|Sequence
    (:then_0
      (Rendering|SetVisibility (Variables|Default|GetFlare) _isactive false)
      (Rendering|SetVisibility (Variables|Default|GetPointLight) _isactive false)
      (Rendering|Components|Light|SetIntensity (Variables|Default|GetPointLight) 25.0)
      (if _isactive
        (if (not (SoundLibrary|IsEventActive _soundlibrarycomponent_vividrazorvault _alarmsoundlibrarytag))
          (SoundLibrary|PlayEvent _soundlibrarycomponent_vividrazorvault _alarmsoundlibrarytag))
        (else
          (SoundLibrary|StopEvent (Variables|Default|GetSoundLibraryComponent_VividRazorVault) (Variables|Default|GetAlarmSoundLibraryTag)))))
    (:then_1
      ; Material Overrides
      (Rendering|Material|SetVectorParameterValue _material_instance_dynamic "SFX_Emissive_Tint(Expand)" (Math|Conversions|ToLinearColor(Float) (Math|Float|SelectFloat 15.0 0.0 _isactive)) :!class /Script/Engine.MaterialInstanceDynamic)
      (Rendering|Material|SetVectorParameterValue _material_instance_dynamic "SFX_Emissive_Tint(Expand)_2" (Math|Conversions|ToLinearColor(Float) (Math|Float|SelectFloat 15.0 0.0 _isactive)) :!class /Script/Engine.MaterialInstanceDynamic)
      (Rendering|Material|SetScalarParameterValue _material_instance_dynamic "Emissive Time Speed Layer 2" (Math|Float|SelectFloat 2.0 0.0 _isactive) :!class /Script/Engine.MaterialInstanceDynamic)
      (Rendering|Material|SetScalarParameterValue _material_instance_dynamic "SFX Emissive Pan Speed X" (Math|Float|SelectFloat 0.75 0.0 _isactive) :!class /Script/Engine.MaterialInstanceDynamic))
    (:then_2
      (return _isactive))))

(fn ConstructionScript
  (|CallParentFunction))
