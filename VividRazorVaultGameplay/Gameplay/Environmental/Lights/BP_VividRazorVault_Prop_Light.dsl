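; Prop light for the VividRazorVault environment that follows time of day and switches to an alert look during lockdown.
; Construction runs the TOD lighting component with a test day phase and caches its brightest light intensity.
; Alert updates recolor every light component: active shows them with a lockdown light function and pauses TOD lighting; inactive clears it, restores cached intensity and reactivates TOD.
; Helpers report default volumetric scattering and select per-phase float values.
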(fn FindBrightestLightTODValue
  (bind _todlightingcomponent (Variables|Default|GetTODLightingComponent))
  (bind _haslightdata (Class|TODLightingComponent|HasLightData _todlightingcomponent))
  (if _haslightdata
    (bind _brightestintensity (Class|TODLightingComponent|GetBrightestLightIntensity _todlightingcomponent))
    (Variables|Default|SetBrightestTODLightIntensity _brightestintensity)))

(fn GetTimeOfDayBlueprintDefaultVariables
  (bind _returnvalue (Actor|GetComponentsByClass "/Script/Engine.LightComponent"))
  (bind _output (Utilities|Array|Get(acopy) _returnvalue))
  (if (not (Utilities|Array|IsEmpty _returnvalue))
    (Utilities|IsValid _output
      (:Is%20Valid
        (return (Utilities|Struct|MakeTimeOfDayBlueprintDefaultVariables 0 (Variables|/Script/Engine.LightComponentBase|GetVolumetricScatteringIntensity _output))))
      (:Is%20Not%20Valid
        _)))
  (return (Utilities|Struct|MakeTimeOfDayBlueprintDefaultVariables 0)))

(fn GetValueFromDayPhaseFloats
  (params (DayPhaseFloats :type Day%20Phase%20Floats%20Structure%20%28by%20ref%29) (Index :type EFortDayPhase%20Enum :default "Morning"))
  (bind (_morning_2_da7d290f4da80ac0d79b888f8b1e63af _day_6_7ac9422d4e5da1f7855ecf8df66bafb1 _evening_7_a7e95df64346d1824eb55e8083a4525f _night_8_bd65e285445a3ed05ca26694cf4e873e) (Utilities|Struct|BreakDayPhaseFloats DayPhaseFloats))
  (return (Utilities|Select :Index Index :Morning _morning_2_da7d290f4da80ac0d79b888f8b1e63af :Day _day_6_7ac9422d4e5da1f7855ecf8df66bafb1 :Evening _evening_7_a7e95df64346d1824eb55e8083a4525f :Night _night_8_bd65e285445a3ed05ca26694cf4e873e)))

(fn UpdateAlertVisuals
  (params (ForceOff :type Boolean) (NewState :type Boolean))
  (bind _isactive (|CallParentFunction ForceOff NewState))
  (for _array_element (Actor|GetComponentsByClass "/Script/Engine.LightComponent")
    (Rendering|Components|Light|SetLightColor _array_element (select _isactive (Variables|Default|GetLightColorAlert) (Variables|Default|GetLightColorDeactivated)) true)
    (Rendering|Components|Light|SetLightFunctionMaterial _array_element (select _isactive (Variables|Default|GetLockdownLightFunction) 0))
    (if _isactive
      (Rendering|SetVisibility _array_element true false)
      (Rendering|Components|Light|SetIntensity _array_element (Variables|Default|GetBrightestTODLightIntensity))))
  (Components|Activation|SetActive (Variables|Default|GetTODLightingComponent) (not _isactive) false)
  (return _isactive))

(fn ConstructionScript
  (|CallParentFunction)
  (Class|TODLightingComponent|RunConstruct (Variables|Default|GetTODLightingComponent) (Variables|Default|GetTest%20Day%20Phase))
  (CallFunction|FindBrightestLightTODValue))
