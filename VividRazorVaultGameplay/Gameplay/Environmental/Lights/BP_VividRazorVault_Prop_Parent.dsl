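; Parent Blueprint for environmental vault props, such as lights, that switch alert effects with their vault's alert state.
; On world ready it listens on the playspace event router: Alert events matching its VaultLocationTag set AlertState, while Paused or Opened events clear it.
; OnRep_AlertState updates cosmetics after a random delay of up to two seconds to stagger props.
; UpdateAlertVisuals activates or fades audio and returns the active state; construction and BeginPlay force it off.
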(event EventOnWorldReady
  (|CallParentFunction)
  (CallFunction|ListenforAlertEvents))

(event EventBeginPlay
  ;[1 1 1] Force all props to be off on first load
  (|CallParentFunction)
  (Networking|CanExecuteCosmeticEvents
    (:True
      (Utilities|FlowControl|DoOnce false
        (:Completed
          (CallFunction|UpdateAlertVisuals true))))))

(event Custom|AlertStateChanged
  (params (NewState :type Boolean))
  (Utilities|FlowControl|Delay (Math|Random|RandomFloatinRange 0.0 2.0))
  (CallFunction|UpdateAlertVisuals false NewState))

(event Custom|ListenForAlertEvents
  (bind (_asynctaskproxy _eventrouter) (Class|AsyncActionWaitforPlayspaceEventRouter|WaitforPlayspaceEventRouter 0 "/VividRazorVaultGameplay/Gameplay/Environmental/Manager/BP_VividRazorVaultPlayspace.BP_VividRazorVaultPlayspace_C" true true)
    (:OnEventRouterAvailable
      (bind (_asynctaskproxy_2 _payload) (GameplayEventRouter|StartListeningToEvent _eventrouter "/VividRazorCoreGameplay/Gameplay/Structs/Struct_VividRazorCore_Vault_Alert_Payload.Struct_VividRazorCore_Vault_Alert_Payload")
        (:then
          (bind (_asynctaskproxy_3 _payload_2) (GameplayEventRouter|StartListeningToEvent _eventrouter "/VividRazorCoreGameplay/Gameplay/Structs/Struct_VividRazorCore_Vault_Paused_Payload.Struct_VividRazorCore_Vault_Paused_Payload")
            (:then
              (bind (_asynctaskproxy_4 _payload_3) (GameplayEventRouter|StartListeningToEvent _eventrouter "/VividRazorCoreGameplay/Gameplay/Structs/Struct_VividRazorCore_Vault_Opened_Payload.Struct_VividRazorCore_Vault_Opened_Payload")
                (:OnEventReceived
                  (if (== (Utilities|Struct|BreakStructVividRazorCoreVaultOpenedPayload _payload_3) (Variables|Default|GetVaultLocationTag))
                    (Variables|Default|SetAlertState false)))))
            (:OnEventReceived
              (if (== (Utilities|Struct|BreakStructVividRazorCoreVaultPausedPayload _payload_2) (Variables|Default|GetVaultLocationTag))
                (Variables|Default|SetAlertState false)))))
        (:OnEventReceived
          (if (== (Utilities|Struct|BreakStructVividRazorCoreVaultAlertPayload _payload) (Variables|Default|GetVaultLocationTag))
            (Variables|Default|SetAlertState true)))))))

(fn OnRep_AlertState
  (Networking|CanExecuteCosmeticEvents
    (:True
      (CallFunction|AlertStateChanged (Variables|Default|GetAlertState)))
    (:False
      _))
  (return))

(fn UpdateAlertVisuals
  (params (ForceOff :type Boolean) (NewState :type Boolean))
  (if ForceOff
    (LocalVariables|SetLocalActive false)
    (else
      (LocalVariables|SetLocalActive NewState)))
  (if (LocalVariables|GetLocalActive)
    ;[1 1 1] Audio
    (bind _audio (Variables|Default|GetAudio)
      (:then
        (Components|Activation|SetActive _audio true false))
      (:else))
    (else
      (bind _audio_2 (Variables|Default|GetAudio)
        (:then
          (Audio|Components|Audio|FadeOut _audio_2 1.0 0.0 "Linear" :!class /Script/Engine.AudioComponent))
        (:else))))
  (return (LocalVariables|GetLocalActive)))

(fn ConstructionScript
  (Networking|CanExecuteCosmeticEvents
    (:True
      (CallFunction|UpdateAlertVisuals true))))
